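# Sound cue for a demolition impact on a phone booth.
# A single phone booth impact wave plays through a Modulator that randomizes pitch on each hit.
# It uses small-object demolition attenuation and the Fort_Impacts sound class, with an object-impact concurrency set capping simultaneous voices at four.
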SoundCue /Game/Sounds/Fort_Impact_Sounds/Demolition/PhoneBooth/Demo_Phonebooth_Impact_Cue
root=Modulator_14
nodes [
 {id=Modulator_14, children [WavePlayer_3], properties {pitchMax=1.2, pitchMin=0.8}}
 {id=WavePlayer_3, properties {soundWaveAssetPtr=&/phonebooth_med_impact_01}}
]
properties {
 attenuationSettings=/Game/Sounds/Attenuation/Fort_Demolition_Hit_Small_Sized_Object_Att
 maxDistance=4000
 soundClassObject=&Fort_Impacts
 volumeMultiplier=0.68
 concurrencyOverrides {maxCount=4}
 concurrencySet [&CONC_Object_Impacts_FourMax]
}
